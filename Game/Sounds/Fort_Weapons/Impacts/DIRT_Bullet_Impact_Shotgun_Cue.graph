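# Sound Cue for shotgun bullet impacts on dirt, mixing two randomized layers.
# One layer picks from three shotgun impact waves, the other from five reused dirt bodyfall waves, mixed louder.
# Each layer gets its own randomized downward pitch modulation, partly offset by a raised cue-wide pitch.
# It uses bullet-impact attenuation, the impacts sound class and shares rifle bullet-impact concurrency.
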SoundCue /Game/Sounds/Fort_Weapons/Impacts/DIRT_Bullet_Impact_Shotgun_Cue
root=Mixer_0
nodes [
 {id=Mixer_0, children [Modulator_0,Modulator_5], properties {inputVolume [1,1.5]}}
 {id=Modulator_0, children [Random_0], properties {pitchMax=0.9, pitchMin=0.8, volumeMax=1.2, volumeMin=1}}
 {id=Modulator_5, children [Random_2], properties {pitchMax=0.8, pitchMin=0.7}}
 {id=Random_0, children [WavePlayer_13, WavePlayer_14, WavePlayer_20], properties {bShouldExcludeFromBranchCulling=true, weights [1,1,1]}}
 {id=Random_2, children [WavePlayer_4,WavePlayer_5,WavePlayer_6,WavePlayer_7,WavePlayer_8], properties {weights [1,1,1,1,1]}}
 {id=WavePlayer_13, properties {soundWaveAssetPtr=&/player_bulletimpact_shotgun_03}}
 {id=WavePlayer_14, properties {soundWaveAssetPtr=&/player_bulletimpact_shotgun_02}}
 {id=WavePlayer_20, properties {soundWaveAssetPtr=&/player_bulletimpact_shotgun_01}}
 {id=WavePlayer_4, properties {soundWaveAssetPtr=/Game/Sounds/Fort_Player/Bodyfalls/Waves/Bodyfall_Dirt_01}}
 {id=WavePlayer_5, properties {soundWaveAssetPtr=/Game/Sounds/Fort_Player/Bodyfalls/Waves/Bodyfall_Dirt_02}}
 {id=WavePlayer_6, properties {soundWaveAssetPtr=/Game/Sounds/Fort_Player/Bodyfalls/Waves/Bodyfall_Dirt_03}}
 {id=WavePlayer_7, properties {soundWaveAssetPtr=/Game/Sounds/Fort_Player/Bodyfalls/Waves/Bodyfall_Dirt_04}}
 {id=WavePlayer_8, properties {soundWaveAssetPtr=/Game/Sounds/Fort_Player/Bodyfalls/Waves/Bodyfall_Dirt_05}}
]
properties {
 attenuationSettings=/Game/Sounds/Attenuation/Fort_Bullet_Impact_Att
 maxDistance=5300
 pitchMultiplier=1.2
 soundClassObject=&Fort_Impacts
 volumeMultiplier=1.062
 concurrencySet [&RifleBulletImpact_Conc]
}
