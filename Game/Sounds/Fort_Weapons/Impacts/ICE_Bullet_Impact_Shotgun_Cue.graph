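# Sound cue for shotgun pellet or bullet impacts on ice, layering a shotgun impact with an ice block hit.
# A mixer plays one of three shotgun impact waves, pitched well up with varied volume, alongside one of five ice block impacts on a short random delay with a milder pitch-up.
# It uses bullet impact attenuation, the impacts sound class and the rifle bullet impact concurrency set.
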SoundCue /Game/Sounds/Fort_Weapons/Impacts/ICE_Bullet_Impact_Shotgun_Cue
root=Mixer_0
nodes [
 {id=Delay_1, children [Modulator_5], properties {delayMax=0.08, delayMin=0.07}}
 {id=Mixer_0, children [Modulator_0,Delay_1], properties {inputVolume [1,1]}}
 {id=Modulator_0, children [Random_0], properties {pitchMax=1.6, pitchMin=1.5, volumeMax=1.2, volumeMin=1}}
 {id=Modulator_5, children [Random_1], properties {pitchMax=1.3, pitchMin=1.2}}
 {id=Random_0, children [WavePlayer_13, WavePlayer_14, WavePlayer_8], properties {bShouldExcludeFromBranchCulling=true, weights [1,1,1]}}
 {id=Random_1, children [WavePlayer_17, WavePlayer_18, WavePlayer_19, WavePlayer_20, WavePlayer_21], properties {weights [1,1,1,1,1]}}
 {id=WavePlayer_13, properties {soundWaveAssetPtr=&/player_bulletimpact_shotgun_03}}
 {id=WavePlayer_14, properties {soundWaveAssetPtr=&/player_bulletimpact_shotgun_02}}
 {id=WavePlayer_17, properties {soundWaveAssetPtr=/Game/Sounds/Fort_Impact_Sounds/Demolition/Ice/IceBlock_Impact_01}}
 {id=WavePlayer_18, properties {soundWaveAssetPtr=/Game/Sounds/Fort_Impact_Sounds/Demolition/Ice/IceBlock_Impact_02}}
 {id=WavePlayer_19, properties {soundWaveAssetPtr=/Game/Sounds/Fort_Impact_Sounds/Demolition/Ice/IceBlock_Impact_03}}
 {id=WavePlayer_20, properties {soundWaveAssetPtr=/Game/Sounds/Fort_Impact_Sounds/Demolition/Ice/IceBlock_Impact_04}}
 {id=WavePlayer_21, properties {soundWaveAssetPtr=/Game/Sounds/Fort_Impact_Sounds/Demolition/Ice/IceBlock_Impact_05}}
 {id=WavePlayer_8, properties {soundWaveAssetPtr=&/player_bulletimpact_shotgun_01}}
]
properties {
 attenuationSettings=/Game/Sounds/Attenuation/Fort_Bullet_Impact_Att
 maxDistance=5300
 soundClassObject=&Fort_Impacts
 volumeMultiplier=1.274
 concurrencySet [&RifleBulletImpact_Conc]
}
