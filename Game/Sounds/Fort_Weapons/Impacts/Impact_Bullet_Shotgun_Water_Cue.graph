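# Sound cue for shotgun bullet impacts on water, using the impacts sound class and bullet impact attenuation.
# A mixer layers two random picks: a medium water bullet impact near natural pitch and a louder-mixed soldier running-in-water footstep pitched well down.
# Each layer passes through its own modulator for slight pitch and volume variation.
# A concurrency override caps simultaneous voices, stopping the oldest.
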SoundCue /Game/Sounds/Fort_Weapons/Impacts/Impact_Bullet_Shotgun_Water_Cue
root=Mixer_0
nodes [
 {id=Mixer_0, children [Modulator_0,Modulator_1], properties {inputVolume [1,1.5]}}
 {id=Modulator_0, children [Random_0], properties {pitchMax=1.15, pitchMin=0.98, volumeMax=1, volumeMin=0.9}}
 {id=Modulator_1, children [Random_1], properties {pitchMax=0.8, pitchMin=0.7, volumeMax=1, volumeMin=0.9}}
 {
  id=Random_0
  children [WavePlayer_0,WavePlayer_1,WavePlayer_2,WavePlayer_3,WavePlayer_4]
  properties {bShouldExcludeFromBranchCulling=true, weights [1,1,1,1,1]}
 }
 {id=Random_1, children [WavePlayer_8, WavePlayer_9, WavePlayer_10, WavePlayer_11, WavePlayer_12], properties {weights [1,1,1,1,1]}}
 {id=WavePlayer_0, properties {soundWaveAssetPtr=/Game/Sounds/Fort_Weapons/Impacts/Bullet_Impacts_S10/Bullet_Impact_Water_Med_10}}
 {id=WavePlayer_1, properties {soundWaveAssetPtr=/Game/Sounds/Fort_Weapons/Impacts/Bullet_Impacts_S10/Bullet_Impact_Water_Med_01}}
 {id=WavePlayer_10, properties {soundWaveAssetPtr=/Game/Sounds/Fort_Step_Sounds/Soldier/Soldier_Step_Run_Water_03}}
 {id=WavePlayer_11, properties {soundWaveAssetPtr=/Game/Sounds/Fort_Step_Sounds/Soldier/Soldier_Step_Run_Water_04}}
 {id=WavePlayer_12, properties {soundWaveAssetPtr=/Game/Sounds/Fort_Step_Sounds/Soldier/Soldier_Step_Run_Water_05}}
 {id=WavePlayer_2, properties {soundWaveAssetPtr=/Game/Sounds/Fort_Weapons/Impacts/Bullet_Impacts_S10/Bullet_Impact_Water_Med_02}}
 {id=WavePlayer_3, properties {soundWaveAssetPtr=/Game/Sounds/Fort_Weapons/Impacts/Bullet_Impacts_S10/Bullet_Impact_Water_Med_03}}
 {id=WavePlayer_4, properties {soundWaveAssetPtr=/Game/Sounds/Fort_Weapons/Impacts/Bullet_Impacts_S10/Bullet_Impact_Water_Med_04}}
 {id=WavePlayer_8, properties {soundWaveAssetPtr=/Game/Sounds/Fort_Step_Sounds/Soldier/Soldier_Step_Run_Water_01}}
 {id=WavePlayer_9, properties {soundWaveAssetPtr=/Game/Sounds/Fort_Step_Sounds/Soldier/Soldier_Step_Run_Water_02}}
]
properties {
 attenuationSettings=/Game/Sounds/Attenuation/Fort_Bullet_Impact_Att
 bOverrideConcurrency=true
 maxDistance=5300
 soundClassObject=&Fort_Impacts
 volumeMultiplier=2.244
 concurrencyOverrides {maxCount=5, resolutionRule=StopOldest, volumeScale=0.5}
 concurrencySet [&RifleBulletImpact_Conc]
}
