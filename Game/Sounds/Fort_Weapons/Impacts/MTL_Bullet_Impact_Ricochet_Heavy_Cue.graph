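# Sound cue for heavy metal bullet impact and ricochet sounds, played in the Impacts sound class.
# A Random node picks among nine equally weighted heavy metal impact waves, excluded from branch culling, and a Modulator varies pitch and volume under a Delay node at the root.
# Bullet impact attenuation handles spatial falloff, and a rifle bullet impact concurrency set governs simultaneous playback.
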SoundCue /Game/Sounds/Fort_Weapons/Impacts/MTL_Bullet_Impact_Ricochet_Heavy_Cue
root=Delay_1
nodes [
 {id=Delay_1, children [Modulator_3]}
 {id=Modulator_3, children [Random_6], properties {pitchMax=1.3, pitchMin=0.9, volumeMax=1, volumeMin=0.8}}
 {
  id=Random_6
  children [WavePlayer_32, WavePlayer_34, WavePlayer_36, WavePlayer_37, WavePlayer_38, WavePlayer_39, WavePlayer_40, WavePlayer_41, WavePlayer_42]
  properties {bShouldExcludeFromBranchCulling=true, weights [1,1,1,1,1,1,1,1,1]}
 }
 {id=WavePlayer_32, properties {soundWaveAssetPtr=/Game/Sounds/Fort_Weapons/Impacts/Bullet_Impacts_S10/Bullet_Impact_Metal_Heavy_09}}
 {id=WavePlayer_34, properties {soundWaveAssetPtr=/Game/Sounds/Fort_Weapons/Impacts/Bullet_Impacts_S10/Bullet_Impact_Metal_Heavy_01}}
 {id=WavePlayer_36, properties {soundWaveAssetPtr=/Game/Sounds/Fort_Weapons/Impacts/Bullet_Impacts_S10/Bullet_Impact_Metal_Heavy_02}}
 {id=WavePlayer_37, properties {soundWaveAssetPtr=/Game/Sounds/Fort_Weapons/Impacts/Bullet_Impacts_S10/Bullet_Impact_Metal_Heavy_03}}
 {id=WavePlayer_38, properties {soundWaveAssetPtr=/Game/Sounds/Fort_Weapons/Impacts/Bullet_Impacts_S10/Bullet_Impact_Metal_Heavy_04}}
 {id=WavePlayer_39, properties {soundWaveAssetPtr=/Game/Sounds/Fort_Weapons/Impacts/Bullet_Impacts_S10/Bullet_Impact_Metal_Heavy_05}}
 {id=WavePlayer_40, properties {soundWaveAssetPtr=/Game/Sounds/Fort_Weapons/Impacts/Bullet_Impacts_S10/Bullet_Impact_Metal_Heavy_06}}
 {id=WavePlayer_41, properties {soundWaveAssetPtr=/Game/Sounds/Fort_Weapons/Impacts/Bullet_Impacts_S10/Bullet_Impact_Metal_Heavy_07}}
 {id=WavePlayer_42, properties {soundWaveAssetPtr=/Game/Sounds/Fort_Weapons/Impacts/Bullet_Impacts_S10/Bullet_Impact_Metal_Heavy_08}}
]
properties {
 attenuationSettings=/Game/Sounds/Attenuation/Fort_Bullet_Impact_Att
 maxDistance=5300
 soundClassObject=&Fort_Impacts
 volumeMultiplier=1.416
 concurrencySet [&RifleBulletImpact_Conc]
}
