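# Sound cue for heavy energy bullet impacts and ricochets on metal, using bullet-impact attenuation, the impacts sound class and rifle-impact concurrency.
# A short fixed delay feeds a quality switch whose first level mixes nine random metal hits with quieter Disintegrator and generic energy layers.
# Each energy layer varies pitch and volume and has an even chance of silence through a null random branch.
# The other three levels share a cheaper three-variant metal branch.
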SoundCue /Game/Sounds/Fort_Weapons/Impacts/MTL_Bullet_Impact_Ricochet_Heavy_Energy_Cue
root=Delay_4
nodes [
 {id=Delay_4, children [QualityLevel_0], properties {delayMax=0.1, delayMin=0.1}}
 {id=Mixer_1, children [Random_6,Modulator_6,Modulator_8], properties {inputVolume [1,0.3,0.5]}}
 {id=Modulator_0, children [Random_8]}
 {id=Modulator_6, children [Random_11], properties {pitchMax=1.1, pitchMin=0.9, volumeMax=1.5}}
 {id=Modulator_8, children [Random_15], properties {pitchMax=1.1, pitchMin=0.9, volumeMax=1.5}}
 {id=QualityLevel_0, children [Mixer_1,Modulator_0,Modulator_0,Modulator_0]}
 {id=Random_10, children [WavePlayer_0,WavePlayer_1], properties {weights [1,1]}}
 {id=Random_11, children [Random_10,null], properties {weights [1,1]}}
 {id=Random_14, children [WavePlayer_4,WavePlayer_5], properties {weights [1,1]}}
 {id=Random_15, children [Random_14,null], properties {weights [1,1]}}
 {
  id=Random_6
  children [WavePlayer_32, WavePlayer_34, WavePlayer_36, WavePlayer_37, WavePlayer_38, WavePlayer_39, WavePlayer_40, WavePlayer_41, WavePlayer_42]
  properties {weights [1,1,1,1,1,1,1,1,1]}
 }
 {id=Random_8, children [WavePlayer_50, WavePlayer_51, WavePlayer_52], properties {bShouldExcludeFromBranchCulling=true, weights [1,1,1]}}
 {id=WavePlayer_0, properties {soundWaveAssetPtr=&/Weapon_Shotgun_VT_Disintegrator_Bullet_Impact_02}}
 {id=WavePlayer_1, properties {soundWaveAssetPtr=&/Weapon_Shotgun_VT_Disintegrator_Bullet_Impact_01}}
 {id=WavePlayer_32, properties {soundWaveAssetPtr=/Game/Sounds/Fort_Weapons/Impacts/Bullet_Impacts_S10/Bullet_Impact_Metal_Heavy_09}}
 {id=WavePlayer_34, properties {soundWaveAssetPtr=/Game/Sounds/Fort_Weapons/Impacts/Bullet_Impacts_S10/Bullet_Impact_Metal_Heavy_01}}
 {id=WavePlayer_36, properties {soundWaveAssetPtr=/Game/Sounds/Fort_Weapons/Impacts/Bullet_Impacts_S10/Bullet_Impact_Metal_Heavy_02}}
 {id=WavePlayer_37, properties {soundWaveAssetPtr=/Game/Sounds/Fort_Weapons/Impacts/Bullet_Impacts_S10/Bullet_Impact_Metal_Heavy_03}}
 {id=WavePlayer_38, properties {soundWaveAssetPtr=/Game/Sounds/Fort_Weapons/Impacts/Bullet_Impacts_S10/Bullet_Impact_Metal_Heavy_04}}
 {id=WavePlayer_39, properties {soundWaveAssetPtr=/Game/Sounds/Fort_Weapons/Impacts/Bullet_Impacts_S10/Bullet_Impact_Metal_Heavy_05}}
 {id=WavePlayer_4, properties {soundWaveAssetPtr=&/VT_Generic_Energy_Bullet_Impact_01}}
 {id=WavePlayer_40, properties {soundWaveAssetPtr=/Game/Sounds/Fort_Weapons/Impacts/Bullet_Impacts_S10/Bullet_Impact_Metal_Heavy_06}}
 {id=WavePlayer_41, properties {soundWaveAssetPtr=/Game/Sounds/Fort_Weapons/Impacts/Bullet_Impacts_S10/Bullet_Impact_Metal_Heavy_07}}
 {id=WavePlayer_42, properties {soundWaveAssetPtr=/Game/Sounds/Fort_Weapons/Impacts/Bullet_Impacts_S10/Bullet_Impact_Metal_Heavy_08}}
 {id=WavePlayer_5, properties {soundWaveAssetPtr=&/VT_Generic_Energy_Bullet_Impact_02}}
 {id=WavePlayer_50, properties {soundWaveAssetPtr=&Bullet_Impact_Metal_Heavy_09}}
 {id=WavePlayer_51, properties {soundWaveAssetPtr=&Bullet_Impact_Metal_Heavy_01}}
 {id=WavePlayer_52, properties {soundWaveAssetPtr=&Bullet_Impact_Metal_Heavy_02}}
]
properties {
 attenuationSettings=/Game/Sounds/Attenuation/Fort_Bullet_Impact_Att
 maxDistance=5300
 soundClassObject=&Fort_Impacts
 volumeMultiplier=0.885
 concurrencySet [&RifleBulletImpact_Conc]
}
