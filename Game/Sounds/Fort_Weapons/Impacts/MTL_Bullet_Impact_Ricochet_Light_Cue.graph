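# Sound cue for light metal bullet impact and ricochet sounds from weapon fire.
# An equal-weight Random node picks one of seven light metal impact waves, excluded from branch culling, then passes through a Modulator for pitch and volume variation and a root Delay node left at default settings.
# Playback uses bullet-impact attenuation, the impacts sound class and a rifle bullet impact concurrency set.
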SoundCue /Game/Sounds/Fort_Weapons/Impacts/MTL_Bullet_Impact_Ricochet_Light_Cue
root=Delay_0
nodes [
 {id=Delay_0, children [Modulator_0]}
 {id=Modulator_0, children [Random_6], properties {pitchMax=1.3, pitchMin=0.9, volumeMax=1, volumeMin=0.8}}
 {
  id=Random_6
  children [WavePlayer_38, WavePlayer_39, WavePlayer_40, WavePlayer_41, WavePlayer_42, WavePlayer_43, WavePlayer_44]
  properties {bShouldExcludeFromBranchCulling=true, weights [1,1,1,1,1,1,1]}
 }
 {id=WavePlayer_38, properties {soundWaveAssetPtr=/Game/Sounds/Fort_Weapons/Impacts/Bullet_Impacts_S10/Bullet_Impact_Metal_Light_01}}
 {id=WavePlayer_39, properties {soundWaveAssetPtr=/Game/Sounds/Fort_Weapons/Impacts/Bullet_Impacts_S10/Bullet_Impact_Metal_Light_02}}
 {id=WavePlayer_40, properties {soundWaveAssetPtr=/Game/Sounds/Fort_Weapons/Impacts/Bullet_Impacts_S10/Bullet_Impact_Metal_Light_03}}
 {id=WavePlayer_41, properties {soundWaveAssetPtr=/Game/Sounds/Fort_Weapons/Impacts/Bullet_Impacts_S10/Bullet_Impact_Metal_Light_04}}
 {id=WavePlayer_42, properties {soundWaveAssetPtr=/Game/Sounds/Fort_Weapons/Impacts/Bullet_Impacts_S10/Bullet_Impact_Metal_Light_05}}
 {id=WavePlayer_43, properties {soundWaveAssetPtr=/Game/Sounds/Fort_Weapons/Impacts/Bullet_Impacts_S10/Bullet_Impact_Metal_Light_06}}
 {id=WavePlayer_44, properties {soundWaveAssetPtr=/Game/Sounds/Fort_Weapons/Impacts/Bullet_Impacts_S10/Bullet_Impact_Metal_Light_07}}
]
properties {
 attenuationSettings=/Game/Sounds/Attenuation/Fort_Bullet_Impact_Att
 maxDistance=5300
 soundClassObject=&Fort_Impacts
 volumeMultiplier=1.416
 concurrencySet [&RifleBulletImpact_Conc]
}
