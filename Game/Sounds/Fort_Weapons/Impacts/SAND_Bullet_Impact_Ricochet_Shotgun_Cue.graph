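# Sound cue for shotgun bullet impacts and ricochets on sand, routed to the impacts sound class with bullet impact attenuation and rifle impact concurrency.
# A mixer layers a randomly chosen shotgun bullet-impact wave, pitched up with slight volume variation, with a randomly chosen sandcastle demolition impact.
# The sand layer plays after a short fixed delay with light pitch variation, so it trails the initial hit.
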SoundCue /Game/Sounds/Fort_Weapons/Impacts/SAND_Bullet_Impact_Ricochet_Shotgun_Cue
root=Mixer_0
nodes [
 {id=Delay_1, children [Modulator_5], properties {delayMax=0.1, delayMin=0.1}}
 {id=Mixer_0, children [Modulator_0,Delay_1], properties {inputVolume [1.2,1.5]}}
 {id=Modulator_0, children [Random_1], properties {pitchMax=1.3, pitchMin=1.2, volumeMax=1.2, volumeMin=1}}
 {id=Modulator_5, children [Random_4], properties {pitchMax=1.1, pitchMin=1}}
 {id=Random_1, children [WavePlayer_13, WavePlayer_14, WavePlayer_20], properties {weights [1,1,1]}}
 {id=Random_4, children [WavePlayer_9, WavePlayer_10, WavePlayer_11, WavePlayer_15], properties {weights [1,1,1,1]}}
 {id=WavePlayer_10, properties {soundWaveAssetPtr=/Game/Sounds/Fort_Impact_Sounds/Demolition/Sandcastle/Demo_SandCastle_Impact_03}}
 {id=WavePlayer_11, properties {soundWaveAssetPtr=/Game/Sounds/Fort_Impact_Sounds/Demolition/Sandcastle/Demo_SandCastle_Impact_02}}
 {id=WavePlayer_13, properties {soundWaveAssetPtr=&/player_bulletimpact_shotgun_03}}
 {id=WavePlayer_14, properties {soundWaveAssetPtr=&/player_bulletimpact_shotgun_02}}
 {id=WavePlayer_15, properties {soundWaveAssetPtr=/Game/Sounds/Fort_Impact_Sounds/Demolition/Sandcastle/Demo_SandCastle_Impact_01}}
 {id=WavePlayer_20, properties {soundWaveAssetPtr=&/player_bulletimpact_shotgun_01}}
 {id=WavePlayer_9, properties {soundWaveAssetPtr=/Game/Sounds/Fort_Impact_Sounds/Demolition/Sandcastle/Demo_SandCastle_Impact_04}}
]
properties {
 attenuationSettings=/Game/Sounds/Attenuation/Fort_Bullet_Impact_Att
 maxDistance=5300
 soundClassObject=&Fort_Impacts
 volumeMultiplier=1.062
 concurrencySet [&RifleBulletImpact_Conc]
}
