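# Sound cue for shotgun bullet impacts on snow, built from reused sand impact, snow bodyfall and Sneaky Snowman sounds.
# A mixer plays three layers at once, each choosing a random wave: lower-pitched sand impacts, raised-pitch snowman destroy or possess sounds, and raised-pitch snow bodyfalls.
# Per-layer modulators vary pitch, with volume variation on the sand layer.
# It uses bullet impact attenuation, the impacts sound class and rifle bullet impact concurrency.
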SoundCue /Game/Sounds/Fort_Weapons/Impacts/SNOW_Bullet_Impact_Shotgun_Cue
root=Mixer_0
nodes [
 {id=Mixer_0, children [Modulator_0,Modulator_5,Modulator_1], properties {inputVolume [1,1.2,0.8]}}
 {id=Modulator_0, children [Random_0], properties {pitchMax=0.9, pitchMin=0.8, volumeMax=1.2, volumeMin=1}}
 {id=Modulator_1, children [Random_1], properties {pitchMax=1.3, pitchMin=1.2}}
 {id=Modulator_5, children [Random_4], properties {pitchMax=1.3, pitchMin=1.2}}
 {
  id=Random_0
  children [WavePlayer_0,WavePlayer_1,WavePlayer_2,WavePlayer_3]
  properties {bShouldExcludeFromBranchCulling=true, weights [1,1,1,1]}
 }
 {id=Random_1, children [WavePlayer_16, WavePlayer_17, WavePlayer_18, WavePlayer_19], properties {weights [1,1,1,1]}}
 {id=Random_4, children [WavePlayer_20, WavePlayer_15], properties {weights [1,1]}}
 {id=WavePlayer_0, properties {soundWaveAssetPtr=/Game/Sounds/Fort_Weapons/Impacts/Bullet_Impacts_S10/Bullet_Impact_Sand_Med_01}}
 {id=WavePlayer_1, properties {soundWaveAssetPtr=/Game/Sounds/Fort_Weapons/Impacts/Bullet_Impacts_S10/Bullet_Impact_Sand_Med_02}}
 {id=WavePlayer_15, properties {soundWaveAssetPtr=/Game/Sounds/Fort_Gadgets/SneakySnowman/Version2/SneakySnowman_Destroy_01}}
 {id=WavePlayer_16, properties {soundWaveAssetPtr=/Game/Sounds/Fort_Player/Bodyfalls/Waves/Bodyfall_Snow_04}}
 {id=WavePlayer_17, properties {soundWaveAssetPtr=/Game/Sounds/Fort_Player/Bodyfalls/Waves/Bodyfall_Snow_03}}
 {id=WavePlayer_18, properties {soundWaveAssetPtr=/Game/Sounds/Fort_Player/Bodyfalls/Waves/Bodyfall_Snow_02}}
 {id=WavePlayer_19, properties {soundWaveAssetPtr=/Game/Sounds/Fort_Player/Bodyfalls/Waves/Bodyfall_Snow_01}}
 {id=WavePlayer_2, properties {soundWaveAssetPtr=/Game/Sounds/Fort_Weapons/Impacts/Bullet_Impacts_S10/Bullet_Impact_Sand_Med_03}}
 {id=WavePlayer_20, properties {soundWaveAssetPtr=/Game/Sounds/Fort_Gadgets/SneakySnowman/Version2/SneakySnowman_Possess_01}}
 {id=WavePlayer_3, properties {soundWaveAssetPtr=/Game/Sounds/Fort_Weapons/Impacts/Bullet_Impacts_S10/Bullet_Impact_Sand_Med_04}}
]
properties {
 attenuationSettings=/Game/Sounds/Attenuation/Fort_Bullet_Impact_Att
 maxDistance=5300
 soundClassObject=&Fort_Impacts
 volumeMultiplier=1.416
 concurrencySet [&RifleBulletImpact_Conc]
}
