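# Sound cue for shotgun impacts on human flesh, routed to the impacts sound class with bullet-impact attenuation.
# A random node picks, with equal weight and repeats allowed, among three shotgun impact waves, each available plain or through an Enveloper that bends its pitch over time.
# The six variants from three samples feed a root Modulator for further variation.
# A concurrency override caps simultaneous instances, stopping the oldest.
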SoundCue /Game/Sounds/Fort_Weapons/Impacts/Shotgun_HumanFlesh_Impact_Cue
root=Modulator_0
nodes [
 {
  id=Enveloper_0
  children [WavePlayer_32]
  properties {pitchCurve {editorCurveData {keys [{time=0}, <at 0.2252 1.4109>, <at 0.456 -0.1395>]}}, volumeCurve {editorCurveData {keys [<at 0 1>]}}}
 }
 {
  id=Enveloper_2
  children [WavePlayer_33]
  properties {
   pitchCurve {editorCurveData {keys [{time=0}, <at 0.0897 1.4972>, <at 0.254 0.6319>, <at 0.7489 0.4609>]}}
   volumeCurve {editorCurveData {keys [<at 0 1>]}}
  }
 }
 {
  id=Enveloper_4
  children [WavePlayer_34]
  properties {pitchCurve {editorCurveData {keys [{time=0}, <at 0.0771 1.4823>, <at 0.4081 1.3073>]}}, volumeCurve {editorCurveData {keys [<at 0 1>]}}}
 }
 {id=Modulator_0, children [Random_0], properties {volumeMax=1.5}}
 {
  id=Random_0
  children [WavePlayer_32, WavePlayer_33, WavePlayer_34, Enveloper_0, Enveloper_2, Enveloper_4]
  properties {bRandomizeWithoutReplacement=false, weights [1,1,1,1,1,1]}
 }
 {id=WavePlayer_32, properties {soundWaveAssetPtr=&/player_bulletimpact_shotgun_01}}
 {id=WavePlayer_33, properties {soundWaveAssetPtr=&/player_bulletimpact_shotgun_02}}
 {id=WavePlayer_34, properties {soundWaveAssetPtr=&/player_bulletimpact_shotgun_03}}
]
properties {
 attenuationSettings=/Game/Sounds/Attenuation/Fort_Bullet_Impact_Att
 bOverrideConcurrency=true
 maxDistance=5300
 soundClassObject=&Fort_Impacts
 volumeMultiplier=0.496
 concurrencyOverrides {maxCount=3, resolutionRule=StopOldest}
}
