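# Sound cue for shotgun bullet impacts on an opponent, classed under Fort_Impacts.
# A quality-level node selects between an eight-wave random pool in its first slot and a five-wave subset in its second, leaving the other two slots empty.
# Each pool feeds a modulator that varies pitch and volume.
# The cue overrides attenuation and concurrency, capping simultaneous plays at four and rejecting new ones beyond that.
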SoundCue /Game/Sounds/Fort_Weapons/Impacts/Shotgun_Opponent_Impact_Cue
root=QualityLevel_0
nodes [
 {id=Modulator_1, children [Random_1], properties {pitchMax=1.2, pitchMin=0.9, volumeMax=1.5}}
 {id=Modulator_2, children [Random_2], properties {pitchMax=1.2, pitchMin=0.9, volumeMax=1.5}}
 {id=QualityLevel_0, children [Modulator_1,Modulator_2,null,null]}
 {
  id=Random_1
  children [WavePlayer_8, WavePlayer_9, WavePlayer_10, WavePlayer_11, WavePlayer_12, WavePlayer_13, WavePlayer_14, WavePlayer_15]
  properties {weights [1,1,1,1,1,1,1,1]}
 }
 {id=Random_2, children [WavePlayer_16, WavePlayer_17, WavePlayer_18, WavePlayer_22, WavePlayer_23], properties {weights [1,1,1,1,1]}}
 {id=WavePlayer_10, properties {soundWaveAssetPtr=&/opponent_bullet_impact_03}}
 {id=WavePlayer_11, properties {soundWaveAssetPtr=&/opponent_bullet_impact_04}}
 {id=WavePlayer_12, properties {soundWaveAssetPtr=&/opponent_bullet_impact_05}}
 {id=WavePlayer_13, properties {soundWaveAssetPtr=&/opponent_bullet_impact_06}}
 {id=WavePlayer_14, properties {soundWaveAssetPtr=&/opponent_bullet_impact_07}}
 {id=WavePlayer_15, properties {soundWaveAssetPtr=&/opponent_bullet_impact_08}}
 {id=WavePlayer_16, properties {soundWaveAssetPtr=&/opponent_bullet_impact_01}}
 {id=WavePlayer_17, properties {soundWaveAssetPtr=&/opponent_bullet_impact_02}}
 {id=WavePlayer_18, properties {soundWaveAssetPtr=&opponent_bullet_impact_03}}
 {id=WavePlayer_22, properties {soundWaveAssetPtr=&opponent_bullet_impact_07}}
 {id=WavePlayer_23, properties {soundWaveAssetPtr=&opponent_bullet_impact_08}}
 {id=WavePlayer_8, properties {soundWaveAssetPtr=&opponent_bullet_impact_01}}
 {id=WavePlayer_9, properties {soundWaveAssetPtr=&opponent_bullet_impact_02}}
]
properties {
 attenuationSettings=/Game/Sounds/Attenuation/Fort_Bullet_ImpactOpponent_Lng_Att
 bOverrideAttenuation=true
 bOverrideConcurrency=true
 maxDistance=4000
 soundClassObject=&Fort_Impacts
 volumeMultiplier=0.708
 concurrencyOverrides {maxCount=4, resolutionRule=PreventNew}
}
