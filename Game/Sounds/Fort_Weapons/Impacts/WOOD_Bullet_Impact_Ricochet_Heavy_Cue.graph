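# Sound cue for heavy bullet impacts and ricochets on wood surfaces.
# An equal-weight Random node picks one of seven heavy wood impact waves, which passes through a Modulator for pitch and volume variation and a root Delay left at default settings.
# A bullet-impact attenuation asset, the impacts sound class and rifle bullet impact concurrency govern playback, with a boosted volume multiplier.
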SoundCue /Game/Sounds/Fort_Weapons/Impacts/WOOD_Bullet_Impact_Ricochet_Heavy_Cue
root=Delay_3
nodes [
 {id=Delay_3, children [Modulator_1]}
 {id=Modulator_1, children [Random_18], properties {pitchMax=1.3, pitchMin=0.9, volumeMax=1, volumeMin=0.8}}
 {
  id=Random_18
  children [WavePlayer_35, WavePlayer_40, WavePlayer_44, WavePlayer_45, WavePlayer_46, WavePlayer_47, WavePlayer_48]
  properties {bShouldExcludeFromBranchCulling=true, weights [1,1,1,1,1,1,1]}
 }
 {id=WavePlayer_35, properties {soundWaveAssetPtr=/Game/Sounds/Fort_Weapons/Impacts/Bullet_Impacts_S10/Bullet_Impact_Wood_Heavy_01}}
 {id=WavePlayer_40, properties {soundWaveAssetPtr=/Game/Sounds/Fort_Weapons/Impacts/Bullet_Impacts_S10/Bullet_Impact_Wood_Heavy_02}}
 {id=WavePlayer_44, properties {soundWaveAssetPtr=/Game/Sounds/Fort_Weapons/Impacts/Bullet_Impacts_S10/Bullet_Impact_Wood_Heavy_03}}
 {id=WavePlayer_45, properties {soundWaveAssetPtr=/Game/Sounds/Fort_Weapons/Impacts/Bullet_Impacts_S10/Bullet_Impact_Wood_Heavy_04}}
 {id=WavePlayer_46, properties {soundWaveAssetPtr=/Game/Sounds/Fort_Weapons/Impacts/Bullet_Impacts_S10/Bullet_Impact_Wood_Heavy_05}}
 {id=WavePlayer_47, properties {soundWaveAssetPtr=/Game/Sounds/Fort_Weapons/Impacts/Bullet_Impacts_S10/Bullet_Impact_Wood_Heavy_06}}
 {id=WavePlayer_48, properties {soundWaveAssetPtr=/Game/Sounds/Fort_Weapons/Impacts/Bullet_Impacts_S10/Bullet_Impact_Wood_Heavy_07}}
]
properties {
 attenuationSettings=/Game/Sounds/Attenuation/Fort_Bullet_Impact_Att
 maxDistance=5300
 soundClassObject=&Fort_Impacts
 volumeMultiplier=1.416
 concurrencySet [&RifleBulletImpact_Conc]
}
